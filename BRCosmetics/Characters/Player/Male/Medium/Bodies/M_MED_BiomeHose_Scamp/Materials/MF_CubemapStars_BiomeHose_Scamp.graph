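# Material function for the BiomeHose Scamp character skin that outputs a view-dependent starfield.
# A ray-traced sphere around the object becomes tiled long-lat UVs, sampling the stars texture with mesh-UV derivatives for mip selection.
# A second sample of that texture uses the Z plane of triplanar virtual-box UVs.
# A contrasted power of the camera vector's absolute vertical component blends toward it, covering the spherical mapping's poles.
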MaterialFunction /BRCosmetics/Characters/Player/Male/Medium/Bodies/M_MED_BiomeHose_Scamp/Materials/MF_CubemapStars_BiomeHose_Scamp
graph {nodes [
  Constant()
  n1 = ObjectPositionWS()
  n2 = ScalarParameter() {defaultValue=100, desc=Approximate radius of the cubemap sphere, parameterName=Sphere Radius}
  n3 = Multiply(A=CameraVectorWS()) {constB=-1}
  n4 = WorldPosition()
  n5 = MaterialFunctionCall!(Vector=Normalize(VectorInput=MaterialFunctionCall!(Sphere Position=@n1, Sphere Radius=@n2, RayDirection=@n3, World position=@n4.XYZ) {materialFunction=/Engine/Functions/Engine_MaterialFunctions02/RayTracedSphere}."Projected Capture Vector ")) {materialFunction=/Engine/Functions/Engine_MaterialFunctions01/Coordinates/LongLatToUV}
  n6 = Constant2Vector() {g=3, r=5}
  n7 = ScalarParameter() {defaultValue=1.5, parameterName=StarsTile}
  n8 = TextureCoordinate()
  n9 = TextureSample!(UVs=Multiply(A=Multiply(A=@n5.Result, B=@n6), B=@n7), Tex, "DDX(UVs)"=DDX(Value=@n8), "DDY(UVs)"=DDY(Value=@n8), Apply View MipBias) {
   mipValueMode=TMVM_Derivative
   samplerType=SAMPLERTYPE_Masks
   texture=/BRCosmetics/Characters/Player/Male/Medium/Bodies/M_MED_BiomeHose_Scamp/Textures/T_BiomeHose_Scamp_Stars
  }
  n10 = TextureSample!(
   UVs=MaterialFunctionCall!(
    Depth=Constant() {r=100}
    Tile=@n7
    SpeedXY=Constant()
    ActorOffset=Constant3Vector() {constant=<rgba 0 0 90 0>}
    UseLocalCameraVector=StaticBool()
   ) {materialFunction=/Game/Global/MaterialFunctions/MF_TriplanarVirtualBox_UVs}.UVZ
   Tex
   Apply View MipBias
  ) {
   samplerType=SAMPLERTYPE_Masks
   texture=&T_BiomeHose_Scamp_Stars
  }
  n11 = CameraVectorWS()
  n12 = Constant3Vector() {constant=<rgba 0 0 1 0>}
  n13 = Power(Base=Abs(None=DotProduct(A=@n11, B=@n12)), Exp=Constant() {r=3}) {constExponent=10}
  n14 = Constant() {r=1}
  FunctionOutput(None=LinearInterpolate(
    A=@n9.RGB
    B=@n10.RGB
    Alpha=MaterialFunctionCall(In=@n13, Contrast=@n14) {materialFunction=&CheapContrast}.Result
   )) {outputName=Stars}
 ]}
