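# Parent material for the BiomeHose Scamp character body, an anime-style shader using material attributes and the Cloth shading model.
# Shared 2023 anime Pre and Calculations functions feed textures and lighting gradients into a character-specific Post function.
# Cubemap stars, masked by an inverted, selectable FX channel, are added to the shaded color only.
# An IsKicks static switch applies MF_Sock, and the inverted shading curve drives temporal responsiveness.
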Material /BRCosmetics/Characters/Player/Male/Medium/Bodies/M_MED_BiomeHose_Scamp/Materials/M_BiomeHose_Scamp_Parent
graph {
 properties {
  bAutomaticallySetUsageInEditor=false
  bUseMaterialAttributes=true
  bUsedWithClothing=true
  bUsedWithMeshDeformer=true
  bUsedWithMorphTargets=true
  bUsedWithNiagaraMeshParticles=true
  bUsedWithSkeletalMesh=true
  bUsedWithStaticLighting=true
  pixelDepthOffsetMode=PDOM_Legacy
  shadingModel=MSM_Cloth
  usedShadingModels=Cloth
 }
 nodes [
  n1 = MaterialFunctionCall!(CustomUV) {materialFunction=/Game/Global/Materials/Character_Parent_2023/Anime_Parent_2023/MaterialFunctions/MFC_AnimeShader_2023_Pre}
  n2 = MaterialFunctionCall!(CustomLightVector, OverrideNormal_WS, STT=@n1.STT, Normal_Map=@n1.STT, DFL=@n1.STT, FX=@n1.STT) {materialFunction=/Game/Global/Materials/Character_Parent_2023/Anime_Parent_2023/MaterialFunctions/MFC_AnimeShader_2023_Calculations}
  n3 = OneMinus(None=ChannelMaskParameter(None=@n1.FX) {group=Galaxy - General Parameters, parameterName=Stars_MaskChannel, sortPriority=0}.R)
  n4 = MaterialFunctionCall() {materialFunction=&/MF_CubemapStars_BiomeHose_Scamp}
  n5 = MaterialFunctionCall!(
   Shading_Gradient=@n2.Shading_Gradient
   Threshold_Gradient=@n2.Shading_Gradient
   Specular_Gradient=@n2.Shading_Gradient
   Sheen_Mask=@n2.Shading_Gradient
   Thickness=@n2.Shading_Gradient
   DFL_Mask=@n2.Shading_Gradient
   FX=@n2.Shading_Gradient
   Metallic_Gradient=@n2.Shading_Gradient
   RimLight_Gradient=@n2.Shading_Gradient
   NormalBuffer=@n2.Shading_Gradient
   --------
   CINE_LightSources=@n2.Shading_Gradient
   CINE_LightColor=@n2.Shading_Gradient
   ---------
   Color_Lit=@n1.Color_Lit
   Color_Shaded=Add(A=@n1.Color_Shaded, B=Multiply(A=@n3, B=@n4.Stars))
   SRM=@n1.Color_Lit
   Emissive=@n1.Color_Lit
   DFL_Color=@n1.Color_Lit
   Specular_Color=@n1.Color_Lit
   RimLight_Color=@n1.Color_Lit
  ) {materialFunction=&/MFC_BiomeHose_Scamp_Post}
  StaticSwitchParameter(
   True=MaterialFunctionCall!(InOpacity, Material=@n5.Result, Mask) {materialFunction=/Game/Global/Materials/Cosmetics/Kicks/Socks/Materials/MF_Sock}.Result
   False=@n5.Result
  ) {group=Kicks, parameterName=IsKicks}
  TemporalResponsivenessOutput!(None=OneMinus(None=@n5.ShadingCurve))
 ]
}
